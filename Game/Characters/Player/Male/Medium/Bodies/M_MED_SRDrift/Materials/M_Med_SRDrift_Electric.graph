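# Translucent unlit material for electric arcs on the SRDrift character body, usable on skeletal and particle meshes.
# World position offset combines SplineThicken, whose width tapers and thins near the camera, with step-timed panning noise and curl-style vector noise for jitter.
# Pulsing, hue-shifted variant or particle color drives HDR emissive through randomized panning bar masks.
# Opacity combines those masks with rotated beam textures, layered-sine flicker, near-camera fade and optional particle alpha.
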Material /Game/Characters/Player/Male/Medium/Bodies/M_MED_SRDrift/Materials/M_Med_SRDrift_Electric
graph {
 properties {
  bUsedWithMeshDeformer=true
  bUsedWithMeshParticles=true
  bUsedWithMorphTargets=true
  bUsedWithNiagaraMeshParticles=true
  bUsedWithSkeletalMesh=true
  blendMode=BLEND_Translucent
  pixelDepthOffsetMode=PDOM_Legacy
  shadingModel=MSM_Unlit
  twoSided=true
  usedShadingModels=Unlit
 }
 nodes [
  n1 = Subtract(A=ComponentMask(None=TextureCoordinate()) {g=true}, B=Constant()) {constB=0.5}
  n2 = Clamp(None=Multiply(A=@n1, B=OneMinus(None=@n1)))
  n3 = Add(A=TextureSample(UVs=TextureCoordinate() {coordinateIndex=1, vTiling=2}) {texture=/Game/Effects/Fort_Effects/Textures/Ramps/T_Ramp_Smooth_Reverse}.G) {constB=0.6}
  n4 = Power(Base=MaterialFunctionCall(Distance=ScalarParameter() {defaultValue=150, parameterName=NearFade}) {materialFunction=&Near_Camera_Fade}.Result)
  n5 = Multiply(
   A=Multiply(A=@n2, B=Clamp(None=@n3))
   B=LinearInterpolate(
    A=ScalarParameter() {parameterName=SplineThickness_Close}
    B=ScalarParameter() {defaultValue=25, parameterName=SplineThickness}
    Alpha=Clamp(None=@n4)
   )
  )
  n6 = TextureCoordinate()
  n7 = TextureCoordinate() {uTiling=0.002, vTiling=0.1}
  n8 = ScalarParameter() {defaultValue=1.5, parameterName=Tile}
  n9 = Add(
   A=StaticSwitchParameter(True=DynamicParameter() {paramNames [RandomOffset,Param2,Param3,Param4]}.RandomOffset, False=Constant()) {parameterName=UseParticleClolorToHandleColor}
   B=Time()
  )
  n10 = ScalarParameter() {defaultValue=25, parameterName=PanStepSize}
  n11 = Divide(A=Round(None=Multiply(A=@n9, B=@n10)), B=@n10)
  n12 = Constant2Vector() {g=1, r=0.25}
  n13 = ScalarParameter() {defaultValue=-1, parameterName=PanSpeed}
  n14 = TextureSample(UVs=Panner(Coordinate=Multiply(A=@n7, B=@n8), Time=@n11, Speed=Multiply(A=@n12, B=@n13)) {bFractionalPart=true, speedX=0.01, speedY=0.1}) {samplerType=SAMPLERTYPE_LinearColor, texture=/Game/Global/Textures/Default/Noise/Noise32x32}
  n15 = OneMinus(None=@n14.RGB)
  n16 = Panner(
   Coordinate=Multiply(A=TextureCoordinate() {vTiling=0.1}, B=Multiply(A=@n8) {constB=6}) {constB=0.1}
   Time=Multiply(A=@n11) {constB=0.5}
   Speed=Multiply(A=@n13) {constB=-1.43}
  ) {bFractionalPart=true, speedX=-0.05}
  n17 = Multiply(
   A=Multiply(A=Subtract(A=@n14.RGB, B=@n15), B=ScalarParameter() {defaultValue=3, parameterName=WPOMult}) {constA=1}
   B=Add(A=TextureSample(UVs=@n16) {texture=/Game/Global/Textures/Default/Masks_and_Gradients/T_Vert_Gradient_Sections}.R) {constB=0.15}
  )
  n18 = Subtract(A=Subtract(A=WorldPosition().XYZ, B=ActorPositionWS()), B=Multiply(A=ScalarParameter() {defaultValue=1, parameterName=CurlSpeed}, B=Time()))
  n19 = ScalarParameter() {defaultValue=25, parameterName=CurlScale}
  n20 = Multiply(
   A=ScalarParameter() {defaultValue=15, parameterName=CurlIntensity}
   B=VectorNoise(World Position=Divide(A=@n18, B=@n19)) {noiseFunction=VNF_VectorALU}
  )
  Add(
   A=Add(
    A=MaterialFunctionCall(WidthBase=@n5, WidthTip=@n5, UVs for Thickness=@n6) {materialFunction=&SplineThicken}.WorldPositionOffset
    B=ShadingPathSwitch(Default=@n17, Deferred=@n17, Mobile=@n17) {desc=Migrated from FeatureLevelSwitch}
   )
   B=Multiply(A=@n2, B=ShadingPathSwitch(Default=@n20, Deferred=@n20, Mobile=@n20) {desc=Migrated from FeatureLevelSwitch})
  )
  Constant()
  Constant() {r=1}
  n21 = LinearInterpolate(
   A=ScalarParameter() {defaultValue=0.3, parameterName=MinColor}
   B=ScalarParameter() {defaultValue=0.8, parameterName=MaxColor}
   Alpha=Clamp(None=Sine(None=@n9) {period=0.5})
  )
  n22 = Clamp(None=@n21)
  n23 = ScalarParameter() {defaultValue=1, parameterName=ColorMultContact}
  n24 = Multiply(A=StaticSwitchParameter(True=ParticleColor().RGB, False=VectorParameter() {parameterName=VariantColor, defaultValue=<rgba 1 0 0.4 1>}.RGB) {parameterName=UseParticleClolorToHandleColor}) {constB=250}
  n25 = ComponentMask(None=TextureCoordinate()) {r=true}
  n26 = Multiply(A=OneMinus(None=@n25), B=@n25)
  n27 = Multiply(A=@n26, B=@n26) {constB=2}
  n28 = LinearInterpolate(A=@n24, B=@n24, Alpha=Clamp(None=Multiply(A=@n27) {constB=4}))
  n29 = ScalarParameter() {defaultValue=25, parameterName=FadeDistance}
  n30 = LinearInterpolate(A=Multiply(A=@n23, B=@n28), B=Multiply(A=@n23, B=@n28), Alpha=Clamp(None=DepthFade(FadeDistance=@n29)))
  n31 = MaterialFunctionCall(
   Hue Shift Percentage=Multiply(A=@n21, B=ScalarParameter() {defaultValue=0.05, parameterName=HueShift_Mult})
   Texture=Multiply(A=@n21, B=ShadingPathSwitch(Default=@n30, Mobile=@n28) {desc=Migrated from FeatureLevelSwitch})
  ) {materialFunction=&HueShift}
  n32 = Multiply(A=TextureCoordinate() {coordinateIndex=1}, B=ScalarParameter() {defaultValue=1, parameterName=MaskTile})
  n33 = Multiply(A=TextureCoordinate() {coordinateIndex=1, uTiling=0}, B=ScalarParameter() {defaultValue=16, parameterName=NumberOfBars})
  n34 = MaterialFunctionCall!(
   PrimaryTextureCoordinates=@n33
   2ndOrderTextureCoordinates=@n33
   RandomSeed=ScalarParameter() {defaultValue=1, parameterName=Speed}
   NoiseDensityBias
   Bypass Noise Influence Crosstalk
  ) {materialFunction=/Game/Effects/Fort_Effects/Materials/_Functions/MF_RandomUVSample}
  n35 = Multiply(A=VectorParameter() {parameterName=ElectricalMask_Speed, defaultValue=<rgba 0 5 0 0>}.RGB, B=@n34.2ndOrderNoiseWithRange)
  n36 = TextureSampleParameter2D(UVs=Panner(Coordinate=@n32, Speed=ComponentMask(None=@n35) {g=true, r=true}) {speedY=0.7}) {parameterName=TexA, samplerSource=SSM_Wrap_WorldGroupSettings, texture=/Game/Global/Textures/Default/Masks_and_Gradients/T_Water_Top_Ramp}
  n37 = Multiply(A=@n34.2ndOrderNoiseWithRange, B=VectorParameter() {parameterName=ElectricalMask_Speed_2, defaultValue=<rgba 0 4.1 0 0>}.RGB)
  n38 = TextureSampleParameter2D(UVs=Panner(Coordinate=@n32, Speed=ComponentMask(None=@n37) {g=true, r=true}) {speedY=1}) {parameterName=TexB, samplerSource=SSM_Wrap_WorldGroupSettings, texture=&T_Water_Top_Ramp}
  n39 = Power(Base=Multiply(A=@n36.G, B=@n38.G), Exp=ScalarParameter() {defaultValue=1, parameterName=MaskContrast})
  MaterialFunctionCall(Emissive=Multiply(
    A=MaterialFunctionCall(LDR Input=@n22, HDR Tint=@n31.Result) {materialFunction=&DeriveHDRfromLDR}.RemappedRange
    B=@n39
   )) {materialFunction=&MF_FX_EmissiveAdjustments}
  n40 = VectorParameter() {parameterName=Electrical_Tiling, defaultValue=<rgba 1 4 0 0>}
  n41 = ComponentMask(None=VectorParameter() {parameterName=Electrical_Speed, defaultValue=<rgba 0 1 0 0>}.RGB) {g=true, r=true}
  n42 = Panner(Coordinate=Multiply(A=TextureCoordinate() {coordinateIndex=1}, B=ComponentMask(None=@n40.RGB) {g=true, r=true}), Time=@n9, Speed=@n41) {speedY=1}
  n43 = ScalarParameter() {defaultValue=0.25, parameterName="Rotation Angle (0-1) (S)"}
  n44 = VectorParameter() {parameterName=Electrical_Tiling_1, defaultValue=<rgba 1 2 0 0>}
  n45 = Panner(Coordinate=Multiply(A=TextureCoordinate() {coordinateIndex=1}, B=ComponentMask(None=@n44.RGB) {g=true, r=true}), Time=@n9, Speed=@n41) {speedY=1}
  n46 = ScalarParameter() {defaultValue=0.25, parameterName="Rotation Angle (0-1) (S)_1"}
  n47 = Add(
   A=TextureSample(UVs=MaterialFunctionCall(UVs=@n42, "Rotation Angle (0-1)"=@n43) {materialFunction=&CustomRotator}.Rotated Values) {samplerType=SAMPLERTYPE_LinearColor, texture=/Game/Effects/Fort_Effects/Textures/Tile/T_AMR_Beam_02}.G
   B=TextureSample(UVs=MaterialFunctionCall(UVs=@n45, "Rotation Angle (0-1)"=@n46) {materialFunction=&CustomRotator}.Rotated Values) {samplerType=SAMPLERTYPE_LinearColor, texture=&T_AMR_Beam_02}.R
  )
  n48 = Multiply(A=Multiply(A=Sine(None=@n9) {period=1.4}, B=Sine(None=@n9) {period=2.1}), B=Sine(None=@n9) {period=2.7})
  n49 = Multiply(A=Power(Base=LinearInterpolate(Alpha=@n48), Exp=ScalarParameter() {defaultValue=3, parameterName=SinePower}) {constExponent=6}) {constB=25}
  n50 = Power(Base=MaterialFunctionCall(Distance=ScalarParameter() {defaultValue=150, parameterName=NearFade}) {materialFunction=&Near_Camera_Fade}.Result) {constExponent=16}
  n51 = Multiply(A=Multiply(A=Multiply(A=@n47, B=@n39), B=Saturate(None=@n49)), B=LinearInterpolate(A=Constant(), B=Constant() {r=1}, Alpha=Clamp(None=@n50)))
  StaticSwitchParameter(True=Multiply(A=@n51, B=ParticleColor().A), False=@n51) {parameterName=UseParticleClolorToHandleOpacity}
  TextureCoordinate() {coordinateIndex=1}
  Time()
 ]
}
